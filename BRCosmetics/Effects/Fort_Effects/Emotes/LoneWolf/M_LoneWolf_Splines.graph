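# Translucent unlit material for the Lone Wolf emote's spline trail effects, usable on Niagara mesh particles.
# Panning smoke, distorted by a swirl texture, multiplies with inverted noise to index a color-curve atlas, then gets desaturation and particle tint.
# Opacity combines trail and sphere masks, a Dynamic Parameter-driven sweep along U, depth fade and particle alpha.
# A camera-to-pixel vector crossed with the tangent builds a camera-facing width offset, tapered between two widths.
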Material /BRCosmetics/Effects/Fort_Effects/Emotes/LoneWolf/M_LoneWolf_Splines
graph {
 properties {
  bUsedWithInstancedStaticMeshes=true
  bUsedWithNiagaraMeshParticles=true
  blendMode=BLEND_Translucent
  pixelDepthOffsetMode=PDOM_Legacy
  shadingModel=MSM_Unlit
  usedShadingModels=Unlit
 }
 nodes [
  ComponentMask(None=TextureCoordinate()) {g=true}
  n1 = LinearInterpolate(A=TextureCoordinate(), B=TextureCoordinate() {coordinateIndex=1}, Alpha=ScalarParameter() {group=UV, parameterName="UV_0,1"})
  n2 = VectorParameter() {group=UV, parameterName=UVTiling, defaultValue=<rgba 0.3 0.55 0.15 0.5>}
  n3 = Multiply(A=@n1, B=AppendVector(A=@n2.R, B=@n2.R))
  n4 = Multiply(A=TextureSample(UVs=Panner(Coordinate=@n3) {speedX=0.3}) {samplerType=SAMPLERTYPE_Normal, texture=/Game/Global/Textures/Default/Flowmaps/T_SwirlFlow_02_N}.R) {constB=0.06}
  n5 = Constant2Vector() {g=1, r=10}
  n6 = Panner(Coordinate=Multiply(A=Add(A=@n1, B=@n4), B=@n5)) {speedX=-5}
  n7 = TextureSample(UVs=@n6) {samplerType=SAMPLERTYPE_Grayscale, texture=&T_Smoke_Trail_Soft}
  n8 = Multiply(
   A=LinearInterpolate(
    A=TextureCoordinate() {uTiling=0.2}
    B=TextureCoordinate() {coordinateIndex=1, uTiling=0.2}
    Alpha=ScalarParameter() {group=UV, parameterName="UV_0,1"}
   )
   B=@n5
  )
  n9 = OneMinus(None=TextureSample(UVs=Panner(Coordinate=@n8) {speedX=-3}) {samplerType=SAMPLERTYPE_Masks, texture=/Game/Weapons/WeaponSkins/Wraps/Textures/Mash/Mash_Noise}.R)
  n10 = CurveAtlasRowParameter(CurveTime=Multiply(A=Multiply(A=@n7.R) {constB=0.8}, B=Multiply(A=@n9) {constB=1.4})) {
   atlas=/Game/Athena/LinearColorCurves/VFX_CurveAtlas
   curve=/Game/Athena/LinearColorCurves/Curve_Contrail_LoneWolf_Color
   defaultValue=50
   group=Color
   parameterName=ColorCurve
  }
  n11 = ScalarParameter() {group=Color, parameterName=Desaturation}
  n12 = ParticleColor()
  MaterialFunctionCall(Emissive=Multiply(A=Desaturation(None=@n10, Fraction=@n11), B=@n12.RGB)) {materialFunction=&MF_FX_EmissiveAdjustments}
  n13 = Subtract(A=CameraPositionWS(), B=WorldPosition().XYZ)
  n14 = MaterialFunctionCall() {materialFunction=&TangentBasis}
  n15 = ScalarParameter() {defaultValue=25, parameterName=Width_1}
  n16 = ScalarParameter() {defaultValue=25, parameterName=Width_2}
  n17 = ComponentMask(None=TextureCoordinate() {desc=UV_Thickness}) {r=true}
  n18 = Add(A=Constant() {r=-0.5}, B=TextureCoordinate() {desc=UV_Thickness})
  Multiply(
   A=Normalize(VectorInput=CrossProduct(A=@n13, B=@n14.X U))
   B=Multiply(A=LinearInterpolate(A=@n15, B=@n16, Alpha=@n17), B=ComponentMask(None=@n18) {g=true})
  )
  n19 = SphereMask(A=TextureCoordinate() {uTiling=0.5}, B=Constant2Vector() {g=0.5}) {attenuationRadius=0.5, hardnessPercent=50}
  n20 = TextureCoordinate() {uTiling=0}
  n21 = Constant2Vector() {g=0.5}
  n22 = LinearInterpolate(
   A=TextureSample(UVs=@n6) {texture=/Game/Effects/Fort_Effects/Textures/Smoke/T_Lobber_Trail_Mask}.R
   B=Power(Base=SphereMask(A=@n20, B=@n21) {attenuationRadius=0.3, hardnessPercent=0}) {constExponent=3}
  )
  n23 = LinearInterpolate(Alpha=DynamicParameter() {defaultValue=<rgba 1 1 25 25>}.Param1) {constA=-0.5, constB=2}
  n24 = SphereMask(
   A=TextureCoordinate() {vTiling=0}
   B=AppendVector(A=Clamp(None=@n23) {maxDefault=1.5, minDefault=-0.5}, B=Constant())
   Radius=ScalarParameter() {defaultValue=0.25}
  ) {hardnessPercent=0}
  n25 = Multiply(
   A=Multiply(A=Divide(A=@n22) {constB=0.01}, B=SphereMask(A=@n20, B=@n21) {attenuationRadius=0.5, hardnessPercent=50})
   B=Multiply(A=Multiply(A=@n24, B=@n24), B=Multiply(A=@n17, B=@n17))
  )
  Multiply(A=DepthFade(Opacity=Multiply(A=@n19, B=@n25)) {fadeDistanceDefault=32}, B=@n12.A)
  n26 = Add(A=Time(), B=VertexColor().R)
  Power(Base=Frac(None=Multiply(A=@n26) {constB=0.25}))
 ]
}
