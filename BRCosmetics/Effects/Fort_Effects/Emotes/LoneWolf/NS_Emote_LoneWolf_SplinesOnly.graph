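# Niagara effect for the Lone Wolf emote, built from timed spline mesh streaks plus early and late pink-red sprite beams with glow sprites.
# Staggered Spline emitters burst meshes colored by User.SplineColorCurve and drive a 0-1 material parameter over particle life; some add brighter, lower-alpha layers.
# User.SplineDelay offsets most spline start times.
# The skeletal-mesh-attached sphere and GPU ember emitters are disabled.
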NiagaraSystem /BRCosmetics/Effects/Fort_Effects/Emotes/LoneWolf/NS_Emote_LoneWolf_SplinesOnly
dependencies {
 usedDataInterfaces [&NiagaraDataInterfaceCurve, &NiagaraDataInterfaceVector2DCurve]
 usedDynamicInputs [&RandomRangeVector, &RotateVectorByQuaternion]
 usedModules [
  &SystemState
  &self:SystemSpawnScript.SystemScriptSource.SystemScriptGraph.NiagaraNodeAssignment_0.SetVariables_#1
 ]
}
emitters {
 Beam {
  emitterUpdateScript [
   &EmitterState(Life Cycle Mode=::NewEnumerator1, Loop Behavior=::NewEnumerator1, Loop Delay=0.5, Loop Duration=1, UseLoopDelay=true)
   &SpawnBurst_Instantaneous(Spawn Time=0.13, Use Spawn Probability=false, UseLoopCountLimit=false)
  ]
  particleSpawnScript [&InitializeParticle(
    Color=<rgba 0.4 0.08 0.1173 0.1>
    Lifetime=5
    Sprite Size=<111 333>
    Sprite Size Mode=::NewEnumerator4
    UsePositionOffset=false
    Write Material Random=true
   )]
  particleUpdateScript [
   &ParticleState()
   &SpriteFacingAndAlignment(Alignment Coordinate Space=::World, Sprite Alignment=<0 0 1>, Sprite Facing=<0 0 0>, WriteAlignment=true, WriteFacing=false)
   &CameraOffset(Camera Offset Amount=99)
   &ScaleSpriteSize(
    Non-Uniform Curve Sprite Scale=DataInterface() {propertyValues {
      XCurve {keys [{tangentMode=RCTM_User, time=0, value=1}, {tangentMode=RCTM_User, time=1, value=1}]}
      YCurve {keys [
        {arriveTangent=1.9423, interpMode=RCIM_Cubic, leaveTangent=1.9423, tangentMode=RCTM_User, time=0, value=0.5}
        {tangentMode=RCTM_User, time=0.7, value=1}
       ]}
     }}
    Scale Sprite Size Mode=::NewEnumerator3
   )
   &ScaleColor(
    Scale Alpha=&FloatFromCurve(FloatCurve=DataInterface() {propertyValues {Curve {keys [
         {arriveTangent=10.3113, interpMode=RCIM_Cubic, leaveTangent=10.3113, tangentMode=RCTM_User, time=0, value=0}
         {arriveTangent=-0.0454, interpMode=RCIM_Cubic, leaveTangent=-0.0454, tangentMode=RCTM_User, time=0.2, value=1}
         {arriveTangent=-2.3405, interpMode=RCIM_Cubic, leaveTangent=-2.3405, tangentMode=RCTM_User, time=1, value=0}
        ]}}})
    ScaleA=true
    ScaleRGB=false
   )
  ]
  rendererClasses [&NiagaraSpriteRendererProperties]
  renderers [{rendererClass=&NiagaraSpriteRendererProperties}]
 }
 Beam001 {
  emitterUpdateScript [
   &EmitterState(Life Cycle Mode=::NewEnumerator1, Loop Behavior=::NewEnumerator1, Loop Delay=4.8, Loop Duration=1, UseLoopDelay=true)
   &SpawnBurst_Instantaneous(Use Spawn Probability=false, UseLoopCountLimit=false)
  ]
  particleSpawnScript [&InitializeParticle(
    Color=<rgba 0.4 0 0.0467 0.04>
    Lifetime=0.8
    Sprite Size=<111 333>
    Sprite Size Mode=::NewEnumerator4
    UsePositionOffset=false
    Write Material Random=true
   )]
  particleUpdateScript [
   &ParticleState()
   &SpriteFacingAndAlignment(Alignment Coordinate Space=::World, Sprite Alignment=<0 0 1>, Sprite Facing=<0 0 0>, WriteAlignment=true, WriteFacing=false)
   &CameraOffset(Camera Offset Amount=99)
   &ScaleSpriteSize(
    Non-Uniform Curve Sprite Scale=DataInterface() {propertyValues {
      XCurve {keys [{tangentMode=RCTM_User, time=0, value=1}, {tangentMode=RCTM_User, time=1, value=1}]}
      YCurve {keys [
        {arriveTangent=1.9423, interpMode=RCIM_Cubic, leaveTangent=1.9423, tangentMode=RCTM_User, time=0, value=0.5}
        {tangentMode=RCTM_User, time=0.7, value=1}
       ]}
     }}
    Scale Sprite Size Mode=::NewEnumerator3
   )
   &ScaleColor(
    Scale Alpha=&FloatFromCurve(
     FloatCurve=DataInterface() {propertyValues {Curve {keys [
         {arriveTangent=10.3113, interpMode=RCIM_Cubic, leaveTangent=10.3113, tangentMode=RCTM_User, time=0, value=0}
         {arriveTangent=-0.0454, interpMode=RCIM_Cubic, leaveTangent=-0.0454, tangentMode=RCTM_User, time=0.05, value=1}
         {arriveTangent=-2.3405, interpMode=RCIM_Cubic, leaveTangent=-2.3405, tangentMode=RCTM_User, time=1, value=0}
        ]}}}
     Scale Curve=10
    )
    ScaleA=true
    ScaleRGB=false
   )
  ]
  rendererClasses [&NiagaraSpriteRendererProperties]
  renderers [{rendererClass=&NiagaraSpriteRendererProperties}]
 }
 Beam_Light {
  emitterUpdateScript [
   &EmitterState(Life Cycle Mode=::NewEnumerator1, Loop Behavior=::NewEnumerator1, Loop Delay=0.5, Loop Duration=1, UseLoopDelay=true)
   &SpawnBurst_Instantaneous(Use Spawn Probability=false, UseLoopCountLimit=false)
  ]
  particleSpawnScript [
   &InitializeParticle(
    Color=<rgba 5 0.9774 1.2761 1>
    Lifetime=5
    Position Offset=<0 0 10>
    Sprite Size Mode=::NewEnumerator3
    Uniform Sprite Size=450
    UsePositionOffset=true
    Write Material Random=true
   )
   &FN_SetSpawnGroup(Number of Spawn Groups=2)
  ]
  particleUpdateScript [
   &ParticleState(Kill Particles When Lifetime Has Elapsed=false, Let Infinitely Lived Particles Die When Emitter Deactivates=true)
   &ScaleColor(
    Scale Alpha=&FloatFromCurve(FloatCurve=DataInterface() {propertyValues {Curve {keys [
         {arriveTangent=10.3113, interpMode=RCIM_Cubic, leaveTangent=10.3113, tangentMode=RCTM_User, time=0.1, value=0}
         {arriveTangent=-0.0454, interpMode=RCIM_Cubic, leaveTangent=-0.0454, tangentMode=RCTM_User, time=0.3, value=1}
         {arriveTangent=-2.3405, interpMode=RCIM_Cubic, leaveTangent=-2.3405, tangentMode=RCTM_User, time=1, value=0}
        ]}}})
    ScaleA=true
    ScaleRGB=false
   )
   &ScaleSpriteSize(
    Non-Uniform Curve Sprite Scale=DataInterface() {propertyValues {
      XCurve {keys [
        {arriveTangent=1.8486, interpMode=RCIM_Cubic, leaveTangent=1.8486, tangentMode=RCTM_User, time=0, value=0.5}
        <at 0.7 1>
        {tangentMode=RCTM_User, time=1, value=1}
       ]}
      YCurve {keys [
        {arriveTangent=1.9423, interpMode=RCIM_Cubic, leaveTangent=1.9423, tangentMode=RCTM_User, time=0, value=0.5}
        {tangentMode=RCTM_User, time=0.7, value=1}
       ]}
     }}
    Scale Sprite Size Mode=::NewEnumerator3
   )
  ]
  rendererClasses [&NiagaraSpriteRendererProperties]
  renderers [{rendererClass=&NiagaraSpriteRendererProperties}]
 }
 Beam_Light001 {
  emitterUpdateScript [
   &EmitterState(Life Cycle Mode=::NewEnumerator1, Loop Behavior=::NewEnumerator1, Loop Delay=4.8, Loop Duration=1, UseLoopDelay=true)
   &SpawnBurst_Instantaneous(Use Spawn Probability=false, UseLoopCountLimit=false)
  ]
  particleSpawnScript [
   &InitializeParticle(
    Color=<rgba 5 0.9774 1.2761 1>
    Position Offset=<0 0 10>
    Sprite Size Mode=::NewEnumerator3
    Uniform Sprite Size=450
    UsePositionOffset=true
    Write Material Random=true
   )
   &FN_SetSpawnGroup(Number of Spawn Groups=2)
  ]
  particleUpdateScript [
   &ParticleState(Kill Particles When Lifetime Has Elapsed=false, Let Infinitely Lived Particles Die When Emitter Deactivates=true)
   &ScaleColor(
    Scale Alpha=&FloatFromCurve(FloatCurve=DataInterface() {propertyValues {Curve {keys [
         {arriveTangent=10.3113, interpMode=RCIM_Cubic, leaveTangent=10.3113, tangentMode=RCTM_User, time=0, value=0}
         {arriveTangent=-0.0454, interpMode=RCIM_Cubic, leaveTangent=-0.0454, tangentMode=RCTM_User, time=0.05, value=1}
         {arriveTangent=-2.3405, interpMode=RCIM_Cubic, leaveTangent=-2.3405, tangentMode=RCTM_User, time=1, value=0}
        ]}}})
    ScaleA=true
    ScaleRGB=false
   )
   &ScaleSpriteSize(
    Non-Uniform Curve Sprite Scale=DataInterface() {propertyValues {
      XCurve {keys [
        {arriveTangent=1.8486, interpMode=RCIM_Cubic, leaveTangent=1.8486, tangentMode=RCTM_User, time=0, value=0.5}
        <at 0.7 1>
        {tangentMode=RCTM_User, time=1, value=1}
       ]}
      YCurve {keys [
        {arriveTangent=1.9423, interpMode=RCIM_Cubic, leaveTangent=1.9423, tangentMode=RCTM_User, time=0, value=0.5}
        {tangentMode=RCTM_User, time=0.7, value=1}
       ]}
     }}
    Scale Sprite Size Mode=::NewEnumerator3
   )
  ]
  rendererClasses [&NiagaraSpriteRendererProperties]
  renderers [{rendererClass=&NiagaraSpriteRendererProperties}]
 }
 FN_AOE_Embers_Center001 {
  bEnabled=false
  simTarget=GPUComputeSim
  emitterUpdateScript [
   &EmitterState(
    Life Cycle Mode=::NewEnumerator1
    Loop Behavior=::NewEnumerator1
    Loop Delay=&Add_Float(A=4.8, B=$User.SplineDelay)
    Loop Duration=1
    UseLoopDelay=true
   )
   &SpawnRate(
    SpawnRate=&FloatFromCurve(FloatCurve=DataInterface() {propertyValues {Curve {keys [<at 0 100>,<at 0.15 100>,<at 0.2 0>]}}}, CurveIndex=$Emitter.NormalizedLoopAge)
    Use Spawn Probability=false
   )
   &SpawnBurst_Instantaneous(Spawn Count=&RandomRangeInt(Minimum=5, Maximum=40), Use Spawn Probability=false)
  ]
  particleSpawnScript [
   /Niagara/Modules/Spawn/Initialization/InitializeParticle(
    Color=<rgba 0.0469 0.0282 0.0152 1>
    Lifetime=&UniformRangedFloat(Minimum=0.25, Maximum=3)
    Module.Write Color=true
    Module.Write Lifetime=true
    Module.Write Mass=false
    Module.Write Material Random=true
    Module.Write Position=true
    Module.Write Scale=false
    Module.Write SpriteRotation=true
    Module.Write SpriteSize=true
    Position=&Add_Vector(A=/Niagara/DynamicInputs/Helpers/SimulationPosition(), B=<0 0 0>)
    Sprite Rotation=&UniformRangedFloat(Minimum=-180, Maximum=180)
    Sprite Size=&Vector2DFromFloat(Value=&UniformRangedFloat(Minimum=1, Maximum=2))
   )
   &SkeletalMeshLocation(Bone Sampling Mode=::NewEnumerator4, MaskSpawnGroup=false, Sampled Position Offset=<0 0 150>, Skeletal Mesh=$User.SkeletalMesh) {enabled=false, module=SkeletalMeshLocation001}
   &AddVelocity(Velocity=&MakeVector(X=0, Y=0, Z=&UniformRangedFloat(Minimum=25, Maximum=75))) {enabled=false}
   /Niagara/Modules/Spawn/Location/SphereLocation(MaskSpawnGroup=false, Module.Override Randomness=false, Offset=<0 0 40>, Sphere Radius=8, Surface Only=false) {enabled=false}
   &SkeletalMeshLocation(
    MaskSpawnGroup=false
    Mesh Sampling Type=::NewEnumerator5
    Sampled Position Offset=<-29 6.5 -10>
    Skeletal Mesh=$User.SkeletalMesh
    Socket Sampling Mode=::NewEnumerator5
   ) {module=SkeletalMeshLocation002}
   /Niagara/Modules/Spawn/Location/V2/SphereLocation(
    MaskSpawnGroup=false
    Override Seed=false
    Sphere Radius=30
    Surface Only=false
    UseNonUniformScale=false
    UseOffset=true
    UseSphereOrientation=false
   ) {module=SphereLocation001}
   &AddVelocityFromPoint(
    Module.Use Falloff Distance=false
    Velocity Strength=&Multiply_Float(A=&RandomRangeFloat(Minimum=200, Maximum=600), B=1.4)
   ) {enabled=false}
   &ApplyInitialForces()
  ]
  particleUpdateScript [
   &ParticleState()
   /Niagara/Modules/Update/Forces/CurlNoiseForce(
    Noise Frequency=0.0125
    Noise Strength=&FloatFromCurve(FloatCurve=DataInterface() {propertyValues {Curve {keys [<at 0 1>,<at 1 0>]}}}, Scale Curve=3000)
    Pan Noise=true
    Pan Noise Field=<0 0 1>
   )
   &AccelerationForce(Acceleration=<0 0 250>)
   &Drag(Drag=8, Use Linear Drag=true, Use Rotational Drag=false)
   &ScaleColor(
    Scale Alpha=&FloatFromCurve(
     FloatCurve=DataInterface() {propertyValues {Curve {keys [
         {arriveTangent=11.9433, interpMode=RCIM_Cubic, leaveTangent=11.9433, tangentMode=RCTM_User, time=0, value=0}
         <at 0.2 1 RCIM_Cubic>
         <at 1 0 RCIM_Cubic>
        ]}}}
     Scale Curve=10
    )
    Scale RGB=&MakeVectorFromLinearColorRGB(Color=&ColorFromCurve(ColorCurve=DataInterface() {propertyValues {
        AlphaCurve {keys [<at 0 1 RCIM_Cubic>]}
        BlueCurve {keys [
          {arriveTangent=0.0135, interpMode=RCIM_Cubic, leaveTangent=0.0135, time=0, value=0}
          <at 0.3156 0.0043>
          {arriveTangent=-0.0062, interpMode=RCIM_Cubic, leaveTangent=-0.0062, time=1, value=0}
         ]}
        GreenCurve {keys [{arriveTangent=-0.2495, interpMode=RCIM_Cubic, leaveTangent=-0.2495, time=0, value=0.0787}, <at 0.3156 0>, <at 1 0 RCIM_Cubic>]}
        RedCurve {keys [
          {arriveTangent=-11.934, interpMode=RCIM_Cubic, leaveTangent=-11.934, time=0, value=4}
          <at 0.3156 0.234>
          {arriveTangent=-0.3419, interpMode=RCIM_Cubic, leaveTangent=-0.3419, time=1, value=0}
         ]}
       }}))
    ScaleA=true
    ScaleRGB=true
   )
   &DynamicMaterialParameters(
    Index 0 Param 1=&FloatFromCurve(FloatCurve=DataInterface() {propertyValues {Curve {keys [<at 0 0>,<at 1 1>]}}})
    Index 0 Param 2=1
    Index 0 Param 3=1
    Index 0 Param 4=1
    Module.Param0WriteEnabled=true
    Module.Param1WriteEnabled=true
    Module.Param2WriteEnabled=true
    Module.Param3WriteEnabled=true
   )
   &ScaleSpriteSize(Scale Factor=&Vector2DFromFloat(Value=&FloatFromCurve(FloatCurve=DataInterface() {propertyValues {Curve {keys [
          {arriveTangent=9.1531, interpMode=RCIM_Cubic, leaveTangent=9.1531, tangentMode=RCTM_User, time=0, value=0.02}
          {arriveTangent=0.48, interpMode=RCIM_Cubic, leaveTangent=0.48, time=0.18, value=0.99}
          <at 1 0 RCIM_Cubic>
         ]}}})))
   &SolveForcesAndVelocity(Clamp Velocity=false, Limit Acceleration=false)
   &ScaleSpriteSizeBySpeed(Max Scale Factor=<1 4>, Min Scale Factor=<0.7 2>)
  ]
  rendererClasses [&NiagaraSpriteRendererProperties]
  renderers [{rendererClass=&NiagaraSpriteRendererProperties}]
 }
 Sphere001 {
  bEnabled=false
  emitterUpdateScript [
   &EmitterState(
    Life Cycle Mode=::NewEnumerator1
    Loop Behavior=::NewEnumerator1
    Loop Delay=&Add_Float(A=4.6, B=$User.SplineDelay)
    Loop Duration=1
    UseLoopDelay=true
   )
   &SpawnBurst_Instantaneous(Spawn Time=0.233, Use Spawn Probability=false, UseLoopCountLimit=false)
  ]
  particleSpawnScript [/Niagara/Modules/Spawn/Initialization/V2/InitializeParticle(
    Color=<rgba 0.6 0.0273 0 0.8>
    Position Offset=<0 20 170>
    Position Offset Coordinate Space=::Local
    Sprite Size Mode=::NewEnumerator3
    Uniform Sprite Size=222
    UsePositionOffset=true
    Write Material Random=true
   )]
  particleUpdateScript [
   &ParticleState()
   &SkeletalMeshLocation(
    MaskSpawnGroup=false
    Mesh Sampling Type=::NewEnumerator5
    Sampled Position Offset=<-29 6.5 -10>
    Skeletal Mesh=$User.SkeletalMesh
    Socket Sampling Mode=::NewEnumerator5
   )
   &ScaleSpriteSize(Uniform Curve Sprite Scale=DataInterface() {propertyValues {Curve {keys [
        {arriveTangent=4.1785, interpMode=RCIM_Cubic, leaveTangent=4.1785, tangentMode=RCTM_User, time=0, value=0}
        {arriveTangent=0.9995, interpMode=RCIM_Cubic, leaveTangent=0.9995, tangentMode=RCTM_User, time=0.2223, value=0.6825}
        {arriveTangent=-0.0405, interpMode=RCIM_Cubic, leaveTangent=-0.0405, tangentMode=RCTM_User, time=1, value=1}
       ]}}})
   &ScaleColor(
    Scale Alpha=&FloatFromCurve(FloatCurve=DataInterface() {propertyValues {Curve {keys [
         {arriveTangent=-2.9427, interpMode=RCIM_Cubic, leaveTangent=-2.9427, leaveTangentWeight=0.9539, tangentMode=RCTM_User, time=0, value=1}
         {arriveTangentWeight=1.1112, interpMode=RCIM_Cubic, tangentMode=RCTM_User, time=1, value=0}
        ]}}})
    ScaleA=true
    ScaleRGB=true
   )
  ]
  rendererClasses [&NiagaraSpriteRendererProperties]
  renderers [{rendererClass=&NiagaraSpriteRendererProperties}]
 }
 Spline {
  emitterSpawnScript [/SetVariables(Emitter.Width=90) {isSetParameters=true}]
  emitterUpdateScript [
   &EmitterState(Life Cycle Mode=::NewEnumerator1, Loop Behavior=::NewEnumerator1, Loop Delay=1.6, Loop Duration=0.5, UseLoopDelay=true)
   /SetVariables(Emitter.Width:NiagaraFloat=&FloatFromCurve(FloatCurve=DataInterface() {propertyValues {Curve {keys [<at 0 1>,<at 1 0>]}}}, CurveIndex=$Emitter.NormalizedLoopAge, Scale Curve=50)) {isSetParameters=true}
   &SpawnBurst_Instantaneous(Use Spawn Probability=false, UseLoopCountLimit=false)
  ]
  particleSpawnScript [&InitializeParticle(
    Lifetime=1.6
    Mesh Scale Mode=::NewEnumerator3
    Position Offset=<0 0 -25>
    Position Offset Coordinate Space=::Local
    UsePositionOffset=true
    Write Material Random=true
   )]
  particleUpdateScript [
   &SolveForcesAndVelocity(Clamp Velocity=false, Limit Acceleration=false)
   &ParticleState()
   &SolveRotationalForcesAndVelocity()
   &DynamicMaterialParameters(
    Index 0 Param 1=&FloatFromCurve(FloatCurve=DataInterface() {propertyValues {Curve {keys [{tangentMode=RCTM_User, time=0, value=0}, {tangentMode=RCTM_User, time=1, value=1}]}}})
    Module.Param0WriteEnabled=true
    Module.Param1WriteEnabled=true
    Module.Param2WriteEnabled=true
    Module.Param3WriteEnabled=true
   )
   &Color(Color=&ColorFromCurve(ColorCurve=$User.SplineColorCurve))
  ]
  rendererClasses [&NiagaraMeshRendererProperties]
  renderers [{rendererClass=&NiagaraMeshRendererProperties, rendererIndex=0}]
 }
 Spline001 {
  emitterSpawnScript [/SetVariables(Emitter.Width=90) {isSetParameters=true}]
  emitterUpdateScript [
   &EmitterState(
    Life Cycle Mode=::NewEnumerator1
    Loop Behavior=::NewEnumerator1
    Loop Delay=&Add_Float(A=1.7, B=$User.SplineDelay)
    Loop Duration=10
    UseLoopDelay=true
   )
   /SetVariables(Emitter.Width:NiagaraFloat=&FloatFromCurve(FloatCurve=DataInterface() {propertyValues {Curve {keys [<at 0 1>,<at 1 0>]}}}, CurveIndex=$Emitter.NormalizedLoopAge, Scale Curve=50)) {isSetParameters=true}
   &SpawnBurst_Instantaneous(Use Spawn Probability=false, UseLoopCountLimit=false)
  ]
  particleSpawnScript [&InitializeParticle(Lifetime=1.6, Mesh Scale Mode=::NewEnumerator3, UsePositionOffset=false, Write Material Random=true)]
  particleUpdateScript [
   &SolveForcesAndVelocity(Clamp Velocity=false, Limit Acceleration=false)
   &ParticleState()
   &SolveRotationalForcesAndVelocity()
   &DynamicMaterialParameters(
    Index 0 Param 1=&FloatFromCurve(FloatCurve=DataInterface() {propertyValues {Curve {keys [{tangentMode=RCTM_User, time=0, value=0}, {tangentMode=RCTM_User, time=1, value=1}]}}})
    Module.Param0WriteEnabled=true
    Module.Param1WriteEnabled=true
    Module.Param2WriteEnabled=true
    Module.Param3WriteEnabled=true
   )
   &Color(Color=&ColorFromCurve(ColorCurve=$User.SplineColorCurve))
  ]
  rendererClasses [&NiagaraMeshRendererProperties]
  renderers [{rendererClass=&NiagaraMeshRendererProperties, rendererIndex=0}]
 }
 Spline002 {
  emitterSpawnScript [/SetVariables(Emitter.Width=90) {isSetParameters=true}]
  emitterUpdateScript [
   &EmitterState(
    Life Cycle Mode=::NewEnumerator1
    Loop Behavior=::NewEnumerator1
    Loop Delay=&Add_Float(A=1.8, B=$User.SplineDelay)
    Loop Duration=10
    UseLoopDelay=true
   )
   /SetVariables(Emitter.Width:NiagaraFloat=&FloatFromCurve(FloatCurve=DataInterface() {propertyValues {Curve {keys [<at 0 1>,<at 1 0>]}}}, CurveIndex=$Emitter.NormalizedLoopAge, Scale Curve=50)) {isSetParameters=true}
   &SpawnBurst_Instantaneous(Use Spawn Probability=false, UseLoopCountLimit=false)
  ]
  particleSpawnScript [&InitializeParticle(Lifetime=1.6, Mesh Scale Mode=::NewEnumerator3, UsePositionOffset=false, Write Material Random=true)]
  particleUpdateScript [
   &SolveForcesAndVelocity(Clamp Velocity=false, Limit Acceleration=false)
   &ParticleState()
   &SolveRotationalForcesAndVelocity()
   &DynamicMaterialParameters(
    Index 0 Param 1=&FloatFromCurve(FloatCurve=DataInterface() {propertyValues {Curve {keys [{tangentMode=RCTM_User, time=0, value=0}, {tangentMode=RCTM_User, time=1, value=1}]}}})
    Module.Param0WriteEnabled=true
    Module.Param1WriteEnabled=true
    Module.Param2WriteEnabled=true
    Module.Param3WriteEnabled=true
   )
   &Color(Color=&ColorFromCurve(ColorCurve=$User.SplineColorCurve))
  ]
  rendererClasses [&NiagaraMeshRendererProperties]
  renderers [{rendererClass=&NiagaraMeshRendererProperties, rendererIndex=0}]
 }
 Spline003 {
  emitterSpawnScript [/SetVariables(Emitter.Width=90) {isSetParameters=true}]
  emitterUpdateScript [
   &EmitterState(
    Life Cycle Mode=::NewEnumerator1
    Loop Behavior=::NewEnumerator1
    Loop Delay=&Add_Float(A=1.9, B=$User.SplineDelay)
    Loop Duration=10
    UseLoopDelay=true
   )
   /SetVariables(Emitter.Width:NiagaraFloat=&FloatFromCurve(FloatCurve=DataInterface() {propertyValues {Curve {keys [<at 0 1>,<at 1 0>]}}}, CurveIndex=$Emitter.NormalizedLoopAge, Scale Curve=50)) {isSetParameters=true}
   &SpawnBurst_Instantaneous(Use Spawn Probability=false, UseLoopCountLimit=false)
  ]
  particleSpawnScript [&InitializeParticle(Lifetime=1.6, Mesh Scale Mode=::NewEnumerator3, Position Offset Coordinate Space=::Local, UsePositionOffset=true, Write Material Random=true)]
  particleUpdateScript [
   &SolveForcesAndVelocity(Clamp Velocity=false, Limit Acceleration=false)
   &ParticleState()
   &SolveRotationalForcesAndVelocity()
   &DynamicMaterialParameters(
    Index 0 Param 1=&FloatFromCurve(FloatCurve=DataInterface() {propertyValues {Curve {keys [{tangentMode=RCTM_User, time=0, value=0}, {tangentMode=RCTM_User, time=1, value=1}]}}})
    Module.Param0WriteEnabled=true
    Module.Param1WriteEnabled=true
    Module.Param2WriteEnabled=true
    Module.Param3WriteEnabled=true
   )
   &Color(Color=&ColorFromCurve(ColorCurve=$User.SplineColorCurve))
  ]
  rendererClasses [&NiagaraMeshRendererProperties]
  renderers [{rendererClass=&NiagaraMeshRendererProperties, rendererIndex=0}]
 }
 Spline004 {
  emitterSpawnScript [/SetVariables(Emitter.Width=90) {isSetParameters=true}]
  emitterUpdateScript [
   &EmitterState(Life Cycle Mode=::NewEnumerator1, Loop Behavior=::NewEnumerator1, Loop Delay=1.6, Loop Duration=10, UseLoopDelay=true)
   &SpawnBurst_Instantaneous(Use Spawn Probability=false, UseLoopCountLimit=false)
   /SetVariables(Emitter.Width:NiagaraFloat=&FloatFromCurve(FloatCurve=DataInterface() {propertyValues {Curve {keys [<at 0 1>,<at 1 0>]}}}, CurveIndex=$Emitter.NormalizedLoopAge, Scale Curve=50)) {isSetParameters=true}
  ]
  particleSpawnScript [&InitializeParticle(
    Lifetime=1.6
    Mesh Scale Mode=::NewEnumerator3
    Position Offset=<0 0 -25>
    Position Offset Coordinate Space=::Local
    UsePositionOffset=true
    Write Material Random=true
   )]
  particleUpdateScript [
   &SolveForcesAndVelocity(Clamp Velocity=false, Limit Acceleration=false)
   &ParticleState()
   &SolveRotationalForcesAndVelocity()
   &DynamicMaterialParameters(
    Index 0 Param 1=&FloatFromCurve(FloatCurve=DataInterface() {propertyValues {Curve {keys [{tangentMode=RCTM_User, time=0, value=0}, {tangentMode=RCTM_User, time=1, value=1}]}}})
    Module.Param0WriteEnabled=true
    Module.Param1WriteEnabled=true
    Module.Param2WriteEnabled=true
    Module.Param3WriteEnabled=true
   )
   &Color(
    Color=&ColorFromCurve(ColorCurve=$User.SplineColorCurve)
    Scale Alpha=0.4
    Scale Color=&VectorFromFloat(Value=4)
   )
  ]
  rendererClasses [&NiagaraMeshRendererProperties]
  renderers [{rendererClass=&NiagaraMeshRendererProperties, rendererIndex=0}]
 }
 Spline005 {
  emitterSpawnScript [/SetVariables(Emitter.Width=90) {isSetParameters=true}]
  emitterUpdateScript [
   &EmitterState(
    Life Cycle Mode=::NewEnumerator1
    Loop Behavior=::NewEnumerator1
    Loop Delay=&Add_Float(A=1.7, B=$User.SplineDelay)
    Loop Duration=10
    UseLoopDelay=true
   )
   &SpawnBurst_Instantaneous(Use Spawn Probability=false, UseLoopCountLimit=false)
   /SetVariables(Emitter.Width:NiagaraFloat=&FloatFromCurve(FloatCurve=DataInterface() {propertyValues {Curve {keys [<at 0 1>,<at 1 0>]}}}, CurveIndex=$Emitter.NormalizedLoopAge, Scale Curve=50)) {isSetParameters=true}
  ]
  particleSpawnScript [&InitializeParticle(Lifetime=1.6, Mesh Scale Mode=::NewEnumerator3, UsePositionOffset=false, Write Material Random=true)]
  particleUpdateScript [
   &SolveForcesAndVelocity(Clamp Velocity=false, Limit Acceleration=false)
   &ParticleState()
   &SolveRotationalForcesAndVelocity()
   &DynamicMaterialParameters(
    Index 0 Param 1=&FloatFromCurve(FloatCurve=DataInterface() {propertyValues {Curve {keys [{tangentMode=RCTM_User, time=0, value=0}, {tangentMode=RCTM_User, time=1, value=1}]}}})
    Module.Param0WriteEnabled=true
    Module.Param1WriteEnabled=true
    Module.Param2WriteEnabled=true
    Module.Param3WriteEnabled=true
   )
   &Color(
    Color=&ColorFromCurve(ColorCurve=$User.SplineColorCurve)
    Scale Alpha=0.4
    Scale Color=&VectorFromFloat(Value=4)
   )
  ]
  rendererClasses [&NiagaraMeshRendererProperties]
  renderers [{rendererClass=&NiagaraMeshRendererProperties, rendererIndex=0}]
 }
 Spline006 {
  emitterSpawnScript [/SetVariables(Emitter.Width=90) {isSetParameters=true}]
  emitterUpdateScript [
   &EmitterState(
    Life Cycle Mode=::NewEnumerator1
    Loop Behavior=::NewEnumerator1
    Loop Delay=&Add_Float(A=1.8, B=$User.SplineDelay)
    Loop Duration=10
    UseLoopDelay=true
   )
   &SpawnBurst_Instantaneous(Use Spawn Probability=false, UseLoopCountLimit=false)
   /SetVariables(Emitter.Width:NiagaraFloat=&FloatFromCurve(FloatCurve=DataInterface() {propertyValues {Curve {keys [<at 0 1>,<at 1 0>]}}}, CurveIndex=$Emitter.NormalizedLoopAge, Scale Curve=50)) {isSetParameters=true}
  ]
  particleSpawnScript [&InitializeParticle(Lifetime=1.6, Mesh Scale Mode=::NewEnumerator3, UsePositionOffset=false, Write Material Random=true)]
  particleUpdateScript [
   &SolveForcesAndVelocity(Clamp Velocity=false, Limit Acceleration=false)
   &ParticleState()
   &SolveRotationalForcesAndVelocity()
   &DynamicMaterialParameters(
    Index 0 Param 1=&FloatFromCurve(FloatCurve=DataInterface() {propertyValues {Curve {keys [{tangentMode=RCTM_User, time=0, value=0}, {tangentMode=RCTM_User, time=1, value=1}]}}})
    Module.Param0WriteEnabled=true
    Module.Param1WriteEnabled=true
    Module.Param2WriteEnabled=true
    Module.Param3WriteEnabled=true
   )
   &Color(
    Color=&ColorFromCurve(ColorCurve=$User.SplineColorCurve)
    Scale Alpha=0.4
    Scale Color=&VectorFromFloat(Value=4)
   )
  ]
  rendererClasses [&NiagaraMeshRendererProperties]
  renderers [{rendererClass=&NiagaraMeshRendererProperties, rendererIndex=0}]
 }
 Spline007 {
  emitterSpawnScript [/SetVariables(Emitter.Width=90) {isSetParameters=true}]
  emitterUpdateScript [
   &EmitterState(
    Life Cycle Mode=::NewEnumerator1
    Loop Behavior=::NewEnumerator1
    Loop Delay=&Add_Float(A=1.9, B=$User.SplineDelay)
    Loop Duration=10
    UseLoopDelay=true
   )
   &SpawnBurst_Instantaneous(Use Spawn Probability=false, UseLoopCountLimit=false)
   /SetVariables(Emitter.Width:NiagaraFloat=&FloatFromCurve(FloatCurve=DataInterface() {propertyValues {Curve {keys [<at 0 1>,<at 1 0>]}}}, CurveIndex=$Emitter.NormalizedLoopAge, Scale Curve=50)) {isSetParameters=true}
  ]
  particleSpawnScript [&InitializeParticle(Lifetime=1.6, Mesh Scale Mode=::NewEnumerator3, Position Offset Coordinate Space=::Local, UsePositionOffset=true, Write Material Random=true)]
  particleUpdateScript [
   &SolveForcesAndVelocity(Clamp Velocity=false, Limit Acceleration=false)
   &ParticleState()
   &SolveRotationalForcesAndVelocity()
   &DynamicMaterialParameters(
    Index 0 Param 1=&FloatFromCurve(FloatCurve=DataInterface() {propertyValues {Curve {keys [{tangentMode=RCTM_User, time=0, value=0}, {tangentMode=RCTM_User, time=1, value=1}]}}})
    Module.Param0WriteEnabled=true
    Module.Param1WriteEnabled=true
    Module.Param2WriteEnabled=true
    Module.Param3WriteEnabled=true
   )
   &Color(
    Color=&ColorFromCurve(ColorCurve=$User.SplineColorCurve)
    Scale Alpha=0.4
    Scale Color=&VectorFromFloat(Value=4)
   )
  ]
  rendererClasses [&NiagaraMeshRendererProperties]
  renderers [{rendererClass=&NiagaraMeshRendererProperties, rendererIndex=0}]
 }
 Spline_1 {
  emitterSpawnScript [/SetVariables(Emitter.Width=90) {isSetParameters=true}]
  emitterUpdateScript [
   &EmitterState(
    Life Cycle Mode=::NewEnumerator1
    Loop Behavior=::NewEnumerator1
    Loop Delay=&Add_Float(A=1.8, B=$User.SplineDelay)
    Loop Duration=1
    Scalability Mode=::NewEnumerator1
    UseLoopDelay=true
   )
   /SetVariables(Emitter.Width:NiagaraFloat=&FloatFromCurve(FloatCurve=DataInterface() {propertyValues {Curve {keys [<at 0 1>,<at 1 0>]}}}, CurveIndex=$Emitter.NormalizedLoopAge, Scale Curve=50)) {isSetParameters=true}
   &SpawnBurst_Instantaneous(Use Spawn Probability=false, UseLoopCountLimit=false)
  ]
  particleSpawnScript [&InitializeParticle(
    Lifetime=0.8
    Mesh Scale Mode=::NewEnumerator3
    Position Offset=<0 -60 35>
    Position Offset Coordinate Space=::Local
    UsePositionOffset=true
    Write Material Random=true
   )]
  particleUpdateScript [
   &SolveForcesAndVelocity(Clamp Velocity=false, Limit Acceleration=false)
   &ParticleState()
   &SolveRotationalForcesAndVelocity()
   &DynamicMaterialParameters(
    Index 0 Param 1=&FloatFromCurve(FloatCurve=DataInterface() {propertyValues {Curve {keys [{tangentMode=RCTM_User, time=0, value=0}, {tangentMode=RCTM_User, time=1, value=1}]}}})
    Module.Param0WriteEnabled=true
    Module.Param1WriteEnabled=true
    Module.Param2WriteEnabled=true
    Module.Param3WriteEnabled=true
   )
   &Color(
    Color=&ColorFromCurve(ColorCurve=$User.SplineColorCurve)
    Scale Alpha=0.5
    Scale Color=&VectorFromFloat(Value=&FloatFromCurve(FloatCurve=DataInterface() {propertyValues {Curve {keys [<at 0 1>,<at 1 0>]}}}, Scale Curve=15))
   )
  ]
  rendererClasses [&NiagaraMeshRendererProperties]
  renderers [{rendererClass=&NiagaraMeshRendererProperties, rendererIndex=0}]
 }
}
summary {
 systemName=NS_Emote_LoneWolf_SplinesOnly
 emitters [
  {emitterName=Spline_1, rendererClasses [&NiagaraMeshRendererProperties]}
  {emitterName=Spline, rendererClasses [&NiagaraMeshRendererProperties]}
  {emitterName=Spline001, rendererClasses [&NiagaraMeshRendererProperties]}
  {emitterName=Spline002, rendererClasses [&NiagaraMeshRendererProperties]}
  {emitterName=Spline003, rendererClasses [&NiagaraMeshRendererProperties]}
  {emitterName=Spline, rendererClasses [&NiagaraMeshRendererProperties]}
  {emitterName=Spline004, rendererClasses [&NiagaraMeshRendererProperties]}
  {emitterName=Spline005, rendererClasses [&NiagaraMeshRendererProperties]}
  {emitterName=Spline006, rendererClasses [&NiagaraMeshRendererProperties]}
  {emitterName=Spline007, rendererClasses [&NiagaraMeshRendererProperties]}
  {emitterName=Beam, rendererClasses [&NiagaraSpriteRendererProperties]}
  {emitterName=Beam_Light, rendererClasses [&NiagaraSpriteRendererProperties]}
  {emitterName=Beam001, rendererClasses [&NiagaraSpriteRendererProperties]}
  {emitterName=Beam_Light001, rendererClasses [&NiagaraSpriteRendererProperties]}
  {bEnabled=false, emitterName=Sphere001, rendererClasses [&NiagaraSpriteRendererProperties]}
  {
   bEnabled=false
   emitterName=FN_AOE_Embers_Center001
   simTarget=GPUComputeSim
   rendererClasses [&NiagaraSpriteRendererProperties]
  }
 ]
 userVariables [
  User.NewNiagaraDataInterfaceSkeletalMesh {
   type=NiagaraDataInterfaceSkeletalMesh
   defaultValue {
    struct=NiagaraExt_VariableValue_DataInterface
    value {dataInterface=&self:NiagaraDataInterfaceSkeletalMesh_1}
   }
  }
  User.RibbonIDOffset {type=NiagaraInt32}
  User.SkeletalMesh {
   type=NiagaraDataInterfaceSkeletalMesh
   defaultValue {
    struct=NiagaraExt_VariableValue_DataInterface
    value {dataInterface=&self:NiagaraDataInterfaceSkeletalMesh_0}
   }
  }
  User.SplineColorCurve {
   type=NiagaraDataInterfaceColorCurve
   description=""
   defaultValue {
    struct=NiagaraExt_VariableValue_DataInterface
    value {
     dataInterface=&self:NiagaraDataInterfaceColorCurve_0
     dataInterfaceClass=&NiagaraDataInterfaceColorCurve
    }
   }
  }
  User.SplineDelay {type=NiagaraFloat, defaultValue=0, description=""}
 ]
}
